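; Base gameplay ability for player abilities that, going by its name, override door bashing, gating activation on the door bash state.
; CanActivateAbility casts the owner's ability system to FortAbilitySystemComponent and allows activation only while no Gameplay.Action.Player.DoorBash ability is active.
; It supplies a no-message failure tag, refuses if the cast fails, and makes no parent call.
; Activate and end events are disabled placeholders with no logic.
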(event Ability|EventActivateAbility
  (|CallParentFunction) ; This node is disabled and will not be called.\nDrag off pins to build functionality.
)

(event Ability|EventOnEndAbility
  (params (bWasCancelled :type Boolean))
  (|CallParentFunction bWasCancelled) ; This node is disabled and will not be called.\nDrag off pins to build functionality.
)

(fn CanActivateAbility
  (params (ActorInfo :type Gameplay%20Ability%20Actor%20Info%20Structure) (Handle :type Gameplay%20Ability%20Spec%20Handle%20Structure))
  (bind (_owneractor _avataractor _playercontroller _abilitysystemcomponent) (Utilities|Struct|BreakGameplayAbilityActorInfo ActorInfo))
  (bind _asfort_ability_system_component (Utilities|Casting|CastToFortAbilitySystemComponent _abilitysystemcomponent)
    (:then
      (return "(GameplayTags=((TagName=\"Athena.Ability.Failed.NoMessage\")))" (not (Abilities|HasActiveAbilitywithAnyMatchingTag _asfort_ability_system_component "(GameplayTags=((TagName=\"Gameplay.Action.Player.DoorBash\")))"))))
    (:CastFailed
      (return 0 false))))
